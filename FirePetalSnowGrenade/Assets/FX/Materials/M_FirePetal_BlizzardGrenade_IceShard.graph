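# Unlit material for Niagara mesh particles, named for the ice shards of the FirePetal blizzard grenade effect.
# A Fresnel term blends pale blue toward near-white at glancing angles, with its exponent driven per particle by the Fresnel_exponent dynamic parameter.
# The blend is multiplied by particle color RGB and fed to the Emissive input of the MF_FX_EmissiveAdjustments function.
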Material /FirePetalSnowGrenade/Assets/FX/Materials/M_FirePetal_BlizzardGrenade_IceShard
graph {
 properties {bAutomaticallySetUsageInEditor=false, bUsedWithNiagaraMeshParticles=true, shadingModel=MSM_Unlit, usedShadingModels=Unlit}
 nodes [
  n1 = Constant3Vector() {constant=<rgba 0.2244 0.4422 0.66 1>}
  n2 = Constant3Vector() {constant=<rgba 0.9 0.9917 1 0>}
  n3 = Fresnel(ExponentIn=DynamicParameter() {defaultValue=<rgba 2 1 1 1>, paramNames [Fresnel_exponent, Param2, Param3, Param4]}.Fresnel_exponent) {baseReflectFraction=0, exponent=2}
  MaterialFunctionCall(Emissive=Multiply(A=LinearInterpolate(A=@n1, B=@n2, Alpha=@n3), B=ParticleColor().RGB)) {materialFunction=&MF_FX_EmissiveAdjustments}
 ]
}
