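# Niagara end effect for the FirePetal snow grenade's blizzard dome, dissipating it with short-lived mesh bursts oriented by a system-level DomeOrientation.
# Dome drives three mesh renderers, with separate color and dynamic-material curves for inner and thick-snow layers.
# Dome_FadeOut follows slightly later, spinning and expanding mostly sideways as its alpha curve rises.
# Time-of-day functions tint both; User.FX_Scale sizes only Dome, as Dome_FadeOut's scale is fixed.
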NiagaraSystem /FirePetalSnowGrenade/Assets/FX/NS_FirePetal_SnowGrenade_BlizzardDome_End
dependencies {
 usedDataInterfaces [&NiagaraDataInterfaceCurve]
 usedDynamicInputs [/Game/Effects/Niagara/Modules/OrientMestToVectorFromCustomRot]
 usedModules [
  &self:SystemSpawnScript.SystemScriptSource.SystemScriptGraph.NiagaraNodeAssignment_0.SetVariables_#1
  &SystemState
 ]
}
emitters {
 Dome {
  emitterUpdateScript [
   &EmitterState()
   &SpawnBurst_Instantaneous(Use Spawn Probability=false, UseLoopCountLimit=false)
  ]
  particleSpawnScript [
   &InitializeParticle(
    Lifetime=0.7
    Mass Mode=::NewEnumerator3
    Mesh Scale Mode=::NewEnumerator3
    Mesh Uniform Scale=&Multiply_Float(A=11, B=$User.FX_Scale)
    Sprite Rotation Mode=::NewEnumerator2
    UsePositionOffset=false
    Write Material Random=true
   )
   &InitialMeshOrientation(
    Mesh Orientation Mode=::NewEnumerator2
    Orientation Coordinate Space=::World
    Rotation=&RandomRangeVector(Minimum=<-0.05 -0.05 -0.05>, Maximum=<0.05 0.05 0.05>)
    Rotation Coordinate Space=::NewEnumerator1
    Use Rotation Vector=true
   ) {enabled=false}
   /SetVariables(Particles.CameraOffsetCore=-120) {isSetParameters=true, module=SetVariables_#2}
   /SetVariables(Particles.MeshOrientation=$System.DomeOrientation) {isSetParameters=true}
  ]
  particleUpdateScript [
   &ParticleState()
   &ScaleMeshSize(Scale Factor=&VectorFromFloat(Value=&FloatFromCurve(FloatCurve=DataInterface() {propertyValues {Curve {keys [{arriveTangent=21.7632, interpMode=RCIM_Cubic, leaveTangent=21.7632, tangentMode=RCTM_User, time=0, value=0.2}, <at 0.1 1 RCIM_Cubic>]}}}))) {enabled=false}
   &ScaleColor(
    Scale Alpha=&FloatFromCurve(
     FloatCurve=DataInterface() {propertyValues {Curve {keys [
         {interpMode=RCIM_Cubic, tangentMode=RCTM_User, time=0.2, value=1}
         {interpMode=RCIM_Cubic, tangentMode=RCTM_User, time=0.6, value=1}
         <at 1 0 RCIM_Cubic>
        ]}}}
     Scale Curve=&FNTimeOfDayScalar(Night Scale=2, Morning Scale=1, Day Scale=2, Evening Scale=1)
    )
    Scale RGB=&MakeVectorFromLinearColorRGB(Color=&MultiplyLinearColor(
      Linear Color=<rgba 0.6 0.8667 1 1>
      Scale Factor=&MakeVector4_FromLinearColor(LinearColor=&FNTimeOfDayColor(
        Night Color=<rgba 0.48 0.5895 0.8 1>
        Morning Color=<rgba 1 0.7463 0.8274 1>
        Day Color=<rgba 1 0.925 0.85 1>
        Evening Color=<rgba 1 0.6607 0.7618 1>
       ))
     ))
    ScaleA=true
    ScaleRGB=true
   )
   &DynamicMaterialParameters(
    Index 0 Param 1=1
    Index 0 Param 2=5
    Index 0 Param 3=&FloatFromCurve(FloatCurve=DataInterface() {propertyValues {Curve {keys [
         {arriveTangent=-6.2582, leaveTangent=-1.25, tangentMode=RCTM_User, time=0.1, value=0}
         {arriveTangent=0.9457, interpMode=RCIM_Cubic, leaveTangent=0.9457, tangentMode=RCTM_User, time=0.6, value=0}
         {arriveTangent=1.435, interpMode=RCIM_Cubic, leaveTangent=1.435, tangentMode=RCTM_User, time=1, value=1}
        ]}}})
    Index 0 Param 4=&Multiply_Float(A=$User.FX_Scale, B=1)
    Module.Index0Param0Write=true
    Module.Index0Param1Write=true
    Module.Index0Param2Write=true
    Module.Index0Param3Write=true
   )
   /SetVariables(Particles.DynMatPar_Inner:Vector4f=&MakeVector4(
     X=1
     Y=13
     Z=&FloatFromCurve(FloatCurve=DataInterface() {propertyValues {Curve {keys [
          {arriveTangent=-1.25, time=0.1, value=0}
          {arriveTangent=0.6539, interpMode=RCIM_Cubic, leaveTangent=0.6539, tangentMode=RCTM_User, time=0.6, value=0}
          {arriveTangent=1.435, interpMode=RCIM_Cubic, leaveTangent=1.435, tangentMode=RCTM_User, time=1, value=1}
         ]}}})
     W=&Multiply_Float(A=$User.FX_Scale, B=2)
    )) {isSetParameters=true}
   /SetVariables(Particles.DynMatPar_SnowThick:Vector4f=&MakeVector4(
     X=1
     Y=2
     Z=&FloatFromCurve(FloatCurve=DataInterface() {propertyValues {Curve {keys [
          {arriveTangent=-0.2555, interpMode=RCIM_Constant, tangentMode=RCTM_User, time=0.1, value=0.73}
          {arriveTangentWeight=0.2033, interpMode=RCIM_Cubic, leaveTangentWeight=0.0467, tangentMode=RCTM_Break, time=0.6, value=0.73}
          {interpMode=RCIM_Cubic, tangentMode=RCTM_User, time=1, value=1}
         ]}}})
     W=&Multiply_Float(A=$User.FX_Scale, B=0.7)
    )) {isSetParameters=true}
   /SetVariables(Particles.ColorSnowThick:LinearColor=&MultiplyLinearColor(
     Linear Color=&FNTimeOfDayColor(Night Color=<rgba 0.7 0.8026 1 1>, Morning Color=<rgba 1 0.7463 0.8274 1>, Day Color=<rgba 1 1 1 1>, Evening Color=<rgba 1 0.6607 0.7618 1>)
     Scale Factor=&MakeVector4(
      X=1
      Y=1
      Z=1
      W=&Multiply_Float(
       A=&FloatFromCurve(FloatCurve=DataInterface() {propertyValues {Curve {keys [
            {interpMode=RCIM_Cubic, tangentMode=RCTM_User, time=0.15, value=1}
            {arriveTangent=-0.0025, interpMode=RCIM_Cubic, leaveTangent=-0.0025, tangentMode=RCTM_User, time=0.6, value=1}
            <at 0.92 0 RCIM_Cubic>
           ]}}})
       B=2.2
      )
     )
    )) {isSetParameters=true}
   /SetVariables(Particles.ColorCore:LinearColor=&MultiplyLinearColor(
     Linear Color=$Particles.Color
     Scale Factor=&MakeFloat4FromVectorAndFloat("Vector (RGB)"=&MakeVectorFromLinearColorRGB(Color=<rgba 0.585 1.014 1.3 1>), "Float (Alpha)"=1)
    )) {isSetParameters=true}
   /SetVariables(Particles.ScaleCore:Vector3f=&Multiply_VectorByFloat(Vector=$Particles.Scale, Float=1)) {enabled=false, isSetParameters=true}
   /SetVariables(Particles.ScaleThickSnow:Vector3f=&Multiply_VectorByFloat(Vector=$Particles.Scale, Float=1)) {isSetParameters=true}
  ]
  rendererClasses [&NiagaraMeshRendererProperties]
  renderers [
   {rendererClass=&NiagaraMeshRendererProperties, rendererIndex=0}
   {rendererClass=&NiagaraMeshRendererProperties, rendererIndex=1}
   {rendererClass=&NiagaraMeshRendererProperties, rendererIndex=2}
  ]
 }
 Dome_FadeOut {
  emitterUpdateScript [
   &EmitterState()
   &SpawnBurst_Instantaneous(Spawn Time=0.1, Use Spawn Probability=false, UseLoopCountLimit=false)
  ]
  particleSpawnScript [
   &InitializeParticle(
    Lifetime=0.8
    Mass Mode=::NewEnumerator3
    Mesh Scale Mode=::NewEnumerator3
    Mesh Uniform Scale=11
    Sprite Rotation Mode=::NewEnumerator2
    UsePositionOffset=false
    Write Material Random=true
   )
   &InitialMeshOrientation(
    Mesh Orientation Mode=::NewEnumerator2
    Orientation Coordinate Space=::World
    Rotation=&RandomRangeVector(Minimum=<-0.05 -0.05 -0.05>, Maximum=<0.05 0.05 0.05>)
    Rotation Coordinate Space=::NewEnumerator1
    Use Rotation Vector=false
   ) {enabled=false}
   /SetVariables(Particles.CameraOffsetCore=-120) {enabled=false, isSetParameters=true, module=SetVariables_#2}
   /SetVariables(Particles.MeshOrientation=$System.DomeOrientation) {isSetParameters=true}
  ]
  particleUpdateScript [
   &ParticleState()
   &ScaleMeshSize(Scale Factor=&MakeVector(
     X=&Lerp_Float(
      A=1
      B=1.3
      Alpha=&FloatFromCurve(FloatCurve=DataInterface() {propertyValues {Curve {keys [
           {arriveTangent=3.5737, interpMode=RCIM_Cubic, leaveTangent=3.5737, tangentMode=RCTM_User, time=0.5, value=0}
           {arriveTangent=0.3151, interpMode=RCIM_Cubic, leaveTangent=0.3151, tangentMode=RCTM_User, time=1, value=1}
          ]}}})
     )
     Y=&Lerp_Float(
      A=1
      B=1.3
      Alpha=&FloatFromCurve(FloatCurve=DataInterface() {propertyValues {Curve {keys [
           {arriveTangent=3.5737, interpMode=RCIM_Cubic, leaveTangent=3.5737, tangentMode=RCTM_User, time=0.5, value=0}
           {arriveTangent=0.3151, interpMode=RCIM_Cubic, leaveTangent=0.3151, tangentMode=RCTM_User, time=1, value=1}
          ]}}})
     )
     Z=&Lerp_Float(
      A=1
      B=1.1
      Alpha=&FloatFromCurve(FloatCurve=DataInterface() {propertyValues {Curve {keys [
           {arriveTangent=3.5737, interpMode=RCIM_Cubic, leaveTangent=3.5737, tangentMode=RCTM_User, time=0.5, value=0}
           {arriveTangent=0.3151, interpMode=RCIM_Cubic, leaveTangent=0.3151, tangentMode=RCTM_User, time=1, value=1}
          ]}}})
     )
    ))
   &UpdateMeshOrientation(
    Rotation Rate=&FloatFromCurve(FloatCurve=DataInterface() {propertyValues {Curve {keys [
         {interpMode=RCIM_Cubic, tangentMode=RCTM_User, time=0, value=0.4}
         {arriveTangent=-4.0881, interpMode=RCIM_Cubic, leaveTangent=-4.0881, tangentMode=RCTM_User, time=1, value=-0.8}
        ]}}})
    Rotation Vector=<0 0 1>
   )
   &ScaleColor(
    Scale Alpha=&FloatFromCurve(
     FloatCurve=DataInterface() {propertyValues {Curve {keys [
         {arriveTangent=0.0947, interpMode=RCIM_Cubic, leaveTangent=0.0947, tangentMode=RCTM_User, time=0, value=0}
         {arriveTangent=2.0013, interpMode=RCIM_Cubic, leaveTangent=2.0013, tangentMode=RCTM_User, time=1, value=1}
        ]}}}
     Scale Curve=&FNTimeOfDayScalar(Night Scale=1.7, Morning Scale=1.8, Day Scale=2.5, Evening Scale=1.8)
    )
    Scale RGB=&MakeVectorFromLinearColorRGB(Color=&MultiplyLinearColor(
      Linear Color=<rgba 1.6634 1.8777 2 1>
      Scale Factor=&MakeVector4_FromLinearColor(LinearColor=&FNTimeOfDayColor(
        Night Color=<rgba 0.48 0.5895 0.8 1>
        Morning Color=<rgba 1 0.7463 0.8274 1>
        Day Color=<rgba 1 0.925 0.85 1>
        Evening Color=<rgba 1 0.6607 0.7618 1>
       ))
     ))
    ScaleA=true
    ScaleRGB=true
   )
   &DynamicMaterialParameters(
    Index 0 Param 1=1
    Index 0 Param 2=4
    Index 0 Param 3=&FloatFromCurve(FloatCurve=DataInterface() {propertyValues {Curve {keys [
         {arriveTangent=1.2383, interpMode=RCIM_Cubic, leaveTangent=1.2383, tangentMode=RCTM_User, time=0.2, value=0.35}
         {arriveTangent=0.9453, interpMode=RCIM_Cubic, leaveTangent=0.9453, tangentMode=RCTM_User, time=1, value=1}
        ]}}})
    Index 0 Param 4=&FloatFromCurve(FloatCurve=DataInterface() {propertyValues {Curve {keys [{interpMode=RCIM_Cubic, tangentMode=RCTM_User, time=0.3, value=0}, {interpMode=RCIM_Cubic, tangentMode=RCTM_User, time=0.7, value=1}]}}})
    Module.Index0Param0Write=true
    Module.Index0Param1Write=true
    Module.Index0Param2Write=true
    Module.Index0Param3Write=true
   )
  ]
  rendererClasses [&NiagaraMeshRendererProperties]
  renderers [
   {rendererClass=&NiagaraMeshRendererProperties, rendererIndex=0}
   {rendererClass=&NiagaraMeshRendererProperties, rendererIndex=1}
  ]
 }
}
summary {
 systemName=NS_FirePetal_SnowGrenade_BlizzardDome_End
 emitters [
  {emitterName=Dome_FadeOut, rendererClasses [&NiagaraMeshRendererProperties]}
  {emitterName=Dome, rendererClasses [&NiagaraMeshRendererProperties]}
 ]
 userVariables [
  User.FX_DomeAngle {type=Vector3f, description="", defaultValue {struct=Vector3f, value=<0 0 0>}}
  User.FX_Scale {type=NiagaraFloat, defaultValue=1, description=""}
 ]
}
