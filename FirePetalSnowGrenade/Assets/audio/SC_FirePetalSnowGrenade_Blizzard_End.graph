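# Sound cue for the end of the FirePetalSnowGrenade's blizzard effect.
# A single wave player feeds the output directly with the Blizzard_End_01 wave, with no randomization or layering.
# The cue plays at reduced volume, borrows the smoke grenade attenuation asset with a set max distance, and uses an Athena outdoor-ducking ambient sound class.
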SoundCue /FirePetalSnowGrenade/Assets/audio/SC_FirePetalSnowGrenade_Blizzard_End
root=WavePlayer_5
nodes [{id=WavePlayer_5, properties {soundWaveAssetPtr=&/FirePetalSnowGrenade_Blizzard_End_01}}]
properties {
 attenuationSettings=/Game/Sounds/Attenuation/Fort_Grenade_Smoke_Att
 maxDistance=6800
 soundClassObject=/Game/Sounds/FortSoundClassesAndModes/Fort_Ambient_DuckOutdoors_Athena_PSM
 volumeMultiplier=0.45
}
