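# Sound cue for the blizzard freeze effect of the FirePetal snow grenade, played under the traps sound class with borrowed close-range melee attenuation.
# A quality-level root selects between a random Freeze plus pitched Activate mix, a two-layer Activate-only mix, and two branches of lightly modulated Freeze_01.
# Modulators vary pitch, plus volume on Activate layers.
# A concurrency override allows one instance, blocking new ones.
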SoundCue /FirePetalSnowGrenade/Assets/audio/SC_FirePetalSnowGrenade_Blizzard_Freeze
root=QualityLevel_0
nodes [
 {id=Mixer_0, children [WavePlayer_1,Modulator_1], properties {inputVolume [0.75,1]}}
 {id=Mixer_1, children [Random_2,Modulator_0], properties {inputVolume [1,1]}}
 {id=Modulator_0, children [WavePlayer_0], properties {pitchMax=1.2, pitchMin=1, volumeMax=1.5, volumeMin=1}}
 {id=Modulator_1, children [WavePlayer_3], properties {pitchMax=1.3, pitchMin=1, volumeMax=1.5, volumeMin=1}}
 {id=Modulator_2, children [WavePlayer_15], properties {pitchMax=1.1, pitchMin=0.9}}
 {id=Modulator_3, children [WavePlayer_16], properties {pitchMax=1.1, pitchMin=0.9}}
 {id=QualityLevel_0, children [Mixer_1,Mixer_0,Modulator_2,Modulator_3]}
 {id=Random_2, children [WavePlayer_12, WavePlayer_13], properties {weights [1,1]}}
 {id=WavePlayer_0, properties {soundWaveAssetPtr=&/FirePetalSnowGrenade_Activate_New}}
 {id=WavePlayer_1, properties {soundWaveAssetPtr=&FirePetalSnowGrenade_Activate_New}}
 {id=WavePlayer_12, properties {soundWaveAssetPtr=&/FirePetalSnowGrenade_Freeze_01}}
 {id=WavePlayer_13, properties {soundWaveAssetPtr=&/FirePetalSnowGrenade_Freeze_02}}
 {id=WavePlayer_15, properties {soundWaveAssetPtr=&FirePetalSnowGrenade_Freeze_01}}
 {id=WavePlayer_16, properties {soundWaveAssetPtr=&FirePetalSnowGrenade_Freeze_01}}
 {id=WavePlayer_3, properties {soundWaveAssetPtr=&FirePetalSnowGrenade_Activate_New}}
]
properties {
 attenuationSettings=/Game/Sounds/Attenuation/Fort_Weapon_Melee_Attack_Close_Range_Stereo_Att
 bOverrideConcurrency=true
 maxDistance=10058.75
 soundClassObject=&Fort_Traps
 volumeMultiplier=0.6
 concurrencyOverrides {maxCount=1, resolutionRule=PreventNew}
}
