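# Looping blizzard ambience SoundCue for the FirePetalSnowGrenade, using grenade-smoke attenuation and an outdoor-ducking ambient sound class.
# Two branches fade in through volume envelopes: three plain wind loops, and a louder branch with its own mono ambient-loop attenuation node.
# The louder branch mixes four loops with endlessly repeating random picks from eight impact sounds.
# A randomized delay sits above the Looping node, so it offsets only the first impact.
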SoundCue /FirePetalSnowGrenade/Assets/audio/SC_FirePetalSnowGrenade_Blizzard_Loop
root=Mixer_3
nodes [
 {id=Attenuation_0, children [Enveloper_2], properties {attenuationSettings=/Game/Sounds/Attenuation/Fort_Item_Ambient_Loop_Mono_Default_Att}}
 {id=Delay_0, children [Looping_0], properties {delayMax=2, delayMin=1.25}}
 {
  id=Enveloper_0
  children [Mixer_2]
  properties {pitchCurve {editorCurveData {keys [{time=0}]}}, volumeCurve {editorCurveData {keys [<at 0 0>,<at 0.75 1>]}}}
 }
 {
  id=Enveloper_2
  children [Mixer_1]
  properties {pitchCurve {editorCurveData {keys [{time=0}]}}, volumeCurve {editorCurveData {keys [<at 0 0>,<at 0.75 1>]}}}
 }
 {id=Looping_0, children [Random_0]}
 {id=Mixer_1, children [WavePlayer_1,WavePlayer_3,WavePlayer_4,WavePlayer_5,Delay_0], properties {inputVolume [0.5,1,1.5,1.8,1.3]}}
 {id=Mixer_2, children [WavePlayer_10, WavePlayer_0, WavePlayer_2], properties {inputVolume [1,1,1.25]}}
 {id=Mixer_3, children [Enveloper_0, Attenuation_0], properties {inputVolume [1,2.5]}}
 {
  id=Random_0
  children [WavePlayer_7, WavePlayer_8, WavePlayer_9, WavePlayer_11, WavePlayer_12, WavePlayer_13, WavePlayer_14, WavePlayer_15]
  properties {weights [1,1,1,1,1,1,1,1]}
 }
 {id=WavePlayer_0, properties {bLooping=true, soundWaveAssetPtr=&/FirePetalSnowGrenade_Blizzard_Loop_02}}
 {id=WavePlayer_1, properties {bLooping=true, soundWaveAssetPtr=&/FirePetalSnowGrenade_Blizzard_Loop_03}}
 {id=WavePlayer_10, properties {bLooping=true, soundWaveAssetPtr=&/FirePetalSnowGrenade_Blizzard_Loop_01}}
 {id=WavePlayer_11, properties {soundWaveAssetPtr=&/FirePetalSnowGrenade_Blizzard_Impact_04}}
 {id=WavePlayer_12, properties {soundWaveAssetPtr=&/FirePetalSnowGrenade_Blizzard_Impact_05}}
 {id=WavePlayer_13, properties {soundWaveAssetPtr=&/FirePetalSnowGrenade_Blizzard_Impact_06}}
 {id=WavePlayer_14, properties {soundWaveAssetPtr=&/FirePetalSnowGrenade_Blizzard_Impact_07}}
 {id=WavePlayer_15, properties {soundWaveAssetPtr=&/FirePetalSnowGrenade_Blizzard_Impact_08}}
 {id=WavePlayer_2, properties {bLooping=true, soundWaveAssetPtr=&/FirePetalSnowGrenade_Blizzard_Loop_04}}
 {id=WavePlayer_3, properties {bLooping=true, soundWaveAssetPtr=&/FirePetalSnowGrenade_Blizzard_Loop_05}}
 {id=WavePlayer_4, properties {bLooping=true, soundWaveAssetPtr=&/FirePetalSnowGrenade_Blizzard_Loop_06}}
 {id=WavePlayer_5, properties {bLooping=true, soundWaveAssetPtr=&/FirePetalSnowGrenade_Blizzard_Loop_07}}
 {id=WavePlayer_7, properties {soundWaveAssetPtr=&/FirePetalSnowGrenade_Blizzard_Impact_01}}
 {id=WavePlayer_8, properties {soundWaveAssetPtr=&/FirePetalSnowGrenade_Blizzard_Impact_02}}
 {id=WavePlayer_9, properties {soundWaveAssetPtr=&/FirePetalSnowGrenade_Blizzard_Impact_03}}
]
properties {
 attenuationSettings=/Game/Sounds/Attenuation/Fort_Grenade_Smoke_Att
 maxDistance=6800
 soundClassObject=/Game/Sounds/FortSoundClassesAndModes/Fort_Ambient_DuckOutdoors_Athena_PSM
}
