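# Sound cue for the Fire Petal Snow Grenade's drop sound.
# A mixer layers a fixed drop wave at half volume with one of three equally weighted equip variations picked at random.
# It uses weapon-drop attenuation and the loot-collect sound class, with overridden concurrency that blocks new instances once a small cap is reached.
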SoundCue /FirePetalSnowGrenade/Assets/audio/SC_FirePetalSnowGrenade_Drop
root=Mixer_1
nodes [
 {id=Mixer_1, children [WavePlayer_0,Random_0], properties {inputVolume [0.5,1]}}
 {id=Random_0, children [WavePlayer_7,WavePlayer_8,WavePlayer_9], properties {weights [1,1,1]}}
 {id=WavePlayer_0, properties {soundWaveAssetPtr=&/FirePetalSnowGrenade_Drop_01}}
 {id=WavePlayer_7, properties {soundWaveAssetPtr=&/FirePetalSnowGrenade_Equip_D_03}}
 {id=WavePlayer_8, properties {soundWaveAssetPtr=&/FirePetalSnowGrenade_Equip_D_02}}
 {id=WavePlayer_9, properties {soundWaveAssetPtr=&/FirePetalSnowGrenade_Equip_D_01}}
]
properties {
 attenuationSettings=&Fort_Weapon_Drop_Att
 bOverrideConcurrency=true
 maxDistance=3000
 soundClassObject=&Fort_Loot_Collect
 volumeMultiplier=0.6
 concurrencyOverrides {maxCount=3, resolutionRule=PreventNew}
}
