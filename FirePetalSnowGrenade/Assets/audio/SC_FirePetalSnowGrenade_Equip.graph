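# Sound cue for equipping the Fire Petal Snow Grenade, played as player foley at reduced volume.
# A single wave player feeds a modulator whose pitch and volume ranges are fixed at unity, so the sound plays without variation.
# It uses close-prop attenuation and overrides concurrency to block new instances rather than cut off playing ones.
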SoundCue /FirePetalSnowGrenade/Assets/audio/SC_FirePetalSnowGrenade_Equip
root=Modulator_1
nodes [
 {id=Modulator_1, children [WavePlayer_1], properties {pitchMax=1, pitchMin=1, volumeMax=1, volumeMin=1}}
 {id=WavePlayer_1, properties {soundWaveAssetPtr=&/FirePetalSnowGrenade_Equip_01}}
]
properties {
 attenuationSettings=/Game/Sounds/Attenuation/Fort_Prop_Close_Att
 bOverrideConcurrency=true
 maxDistance=1800
 soundClassObject=&Fort_Player_Foley
 volumeMultiplier=0.6
 concurrencyOverrides {resolutionRule=PreventNew}
}
