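# Sound cue for the FirePetalSnowGrenade explosion, routed to the large explosives sound class.
# A quality node sends its first level to a five-layer mix of explosion, two impact sets, snow burst and distant tail.
# The other three levels share a lighter three-layer mix with fewer waves.
# Each layer has its own attenuation and volume modulation, most add pitch variation and random wave selection, and overridden concurrency prevents new instances at the limit.
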SoundCue /FirePetalSnowGrenade/Assets/audio/SC_FirePetalSnowGrenade_Explode
root=QualityLevel_0
nodes [
 {id=Attenuation_0, children [Modulator_13], properties {attenuationSettings=/Game/Sounds/Attenuation/Fort_Explosion_Far_Grenade_Att}}
 {id=Attenuation_1, children [Modulator_12], properties {attenuationSettings=/Game/Sounds/Attenuation/Fort_Grenade_Smoke_Att}}
 {id=Attenuation_2, children [Modulator_2], properties {attenuationSettings=&Fort_Grenade_Smoke_Att}}
 {id=Attenuation_3, children [Modulator_3], properties {attenuationSettings=&Fort_Grenade_Smoke_Att}}
 {
  id=Attenuation_4
  children [Modulator_4]
  properties {attenuationSettings=&Fort_Weapon_Projectile_Medium_Mono_Athena_Att}
 }
 {id=Attenuation_6, children [Modulator_5], properties {attenuationSettings=&Fort_Grenade_Smoke_Att}}
 {id=Attenuation_8, children [Modulator_1], properties {attenuationSettings=&Fort_Explosion_Far_Grenade_Att}}
 {id=Attenuation_9, children [Modulator_0], properties {attenuationSettings=/Game/Sounds/Attenuation/Fort_Explosion_Close_Athena_Att}}
 {id=Mixer_0, children [Attenuation_9, Attenuation_4, Attenuation_8], properties {inputVolume [1,1,0.8]}}
 {
  id=Mixer_2
  children [Attenuation_1, Attenuation_3, Attenuation_2, Attenuation_6, Attenuation_0]
  properties {inputVolume [0.8,0.25,0.4,0.25,0.5]}
 }
 {id=Modulator_0, children [Random_2], properties {volumeMax=1.2, volumeMin=1.1}}
 {id=Modulator_1, children [WavePlayer_17], properties {pitchMin=0.8, volumeMax=1.2, volumeMin=1.1}}
 {id=Modulator_12, children [Random_4], properties {volumeMax=1.2, volumeMin=1.1}}
 {id=Modulator_13, children [Random_3], properties {pitchMin=0.8, volumeMax=1.2, volumeMin=1.1}}
 {id=Modulator_2, children [Random_5], properties {pitchMin=0.8, volumeMax=1.2, volumeMin=1.1}}
 {id=Modulator_3, children [Random_6], properties {pitchMin=0.8, volumeMax=1.2, volumeMin=1.1}}
 {id=Modulator_4, children [Random_0], properties {pitchMin=0.8, volumeMax=1.2, volumeMin=1.1}}
 {id=Modulator_5, children [Random_1], properties {pitchMin=0.8, volumeMax=1.2, volumeMin=1.1}}
 {id=QualityLevel_0, children [Mixer_2,Mixer_0,Mixer_0,Mixer_0]}
 {id=Random_0, children [WavePlayer_3,WavePlayer_4], properties {weights [1,1]}}
 {id=Random_1, children [WavePlayer_6,WavePlayer_7], properties {weights [1,1]}}
 {id=Random_2, children [WavePlayer_14, WavePlayer_16], properties {weights [1,1]}}
 {id=Random_3, children [WavePlayer_10, WavePlayer_11, WavePlayer_12], properties {weights [1,1,1]}}
 {id=Random_4, children [WavePlayer_0,WavePlayer_1,WavePlayer_2], properties {weights [1,1,1]}}
 {id=Random_5, children [WavePlayer_19, WavePlayer_20], properties {weights [1,1]}}
 {id=Random_6, children [WavePlayer_22, WavePlayer_23, WavePlayer_24, WavePlayer_25], properties {weights [1,1,1,1]}}
 {id=WavePlayer_0, properties {soundWaveAssetPtr=&/FirePetalSnowGrenade_Explode_01}}
 {id=WavePlayer_1, properties {soundWaveAssetPtr=&/FirePetalSnowGrenade_Explode_02}}
 {id=WavePlayer_10, properties {soundWaveAssetPtr=&/FirePetalSnowGrenade_Explode_Distant_01}}
 {id=WavePlayer_11, properties {soundWaveAssetPtr=&/FirePetalSnowGrenade_Explode_Distant_02}}
 {id=WavePlayer_12, properties {soundWaveAssetPtr=&/FirePetalSnowGrenade_Explode_Distant_03}}
 {id=WavePlayer_14, properties {soundWaveAssetPtr=&FirePetalSnowGrenade_Explode_01}}
 {id=WavePlayer_16, properties {soundWaveAssetPtr=&FirePetalSnowGrenade_Explode_02}}
 {id=WavePlayer_17, properties {soundWaveAssetPtr=&FirePetalSnowGrenade_Explode_Distant_01}}
 {id=WavePlayer_19, properties {soundWaveAssetPtr=&/FirePetalSnowGrenade_Impact_A_02}}
 {id=WavePlayer_2, properties {soundWaveAssetPtr=&/FirePetalSnowGrenade_Explode_03}}
 {id=WavePlayer_20, properties {soundWaveAssetPtr=&/FirePetalSnowGrenade_Impact_A_03}}
 {id=WavePlayer_22, properties {soundWaveAssetPtr=&/FirePetalSnowGrenade_Impact_B_01}}
 {id=WavePlayer_23, properties {soundWaveAssetPtr=&/FirePetalSnowGrenade_Impact_B_02}}
 {id=WavePlayer_24, properties {soundWaveAssetPtr=&/FirePetalSnowGrenade_Impact_B_03}}
 {id=WavePlayer_25, properties {soundWaveAssetPtr=&/FirePetalSnowGrenade_Impact_B_04}}
 {id=WavePlayer_3, properties {soundWaveAssetPtr=&/FirePetalSnowGrenade_Impact_A_01}}
 {id=WavePlayer_4, properties {soundWaveAssetPtr=&FirePetalSnowGrenade_Impact_A_02}}
 {id=WavePlayer_6, properties {soundWaveAssetPtr=&/FirePetalSnowGrenade_SnowExplode_02}}
 {id=WavePlayer_7, properties {soundWaveAssetPtr=&/FirePetalSnowGrenade_SnowExplode_01}}
]
properties {
 bOverrideConcurrency=true
 maxDistance=10579.2871
 soundClassObject=/Game/Sounds/FortSoundClassesAndModes/Fort_Explosives_Large_PSM
 volumeMultiplier=0.8
 concurrencyOverrides {resolutionRule=PreventNew}
}
