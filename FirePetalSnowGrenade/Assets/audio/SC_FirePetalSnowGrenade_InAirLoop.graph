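# Sound cue for the Fire Petal Snow Grenade in flight, pairing throw one-shots with a looping in-air sound.
# A mixer combines the loop, shaped by a volume-ramp envelope and slight modulation, with two throw layers, each randomized over three waves through delay and modulator nodes.
# The root attenuation node uses RPG explosion falloff, while cue properties name projectile-loop attenuation and cap concurrency at three.
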SoundCue /FirePetalSnowGrenade/Assets/audio/SC_FirePetalSnowGrenade_InAirLoop
root=Attenuation_1
nodes [
 {id=Attenuation_1, children [Mixer_1], properties {attenuationSettings=/Game/Sounds/Attenuation/Fort_Explosion_RPG_Att}}
 {id=Delay_0, children [Random_0]}
 {id=Delay_1, children [Random_2]}
 {
  id=Enveloper_1
  children [Modulator_2]
  properties {pitchCurve {editorCurveData {keys [{time=0}]}}, volumeCurve {editorCurveData {keys [<at 0.0704 0.2337>,<at 0.25 0.967>]}}}
 }
 {id=Mixer_1, children [Enveloper_1,Modulator_0,Modulator_1], properties {inputVolume [1,1,1]}}
 {id=Modulator_0, children [Delay_1]}
 {id=Modulator_1, children [Delay_0]}
 {id=Modulator_2, children [WavePlayer_9], properties {volumeMax=1, volumeMin=0.9}}
 {id=Random_0, children [WavePlayer_0,WavePlayer_1,WavePlayer_2], properties {weights [1,1,1]}}
 {id=Random_2, children [WavePlayer_6,WavePlayer_7,WavePlayer_8], properties {weights [1,1,1]}}
 {id=WavePlayer_0, properties {soundWaveAssetPtr=&/FirePetalSnowGrenade_Throw_B_01}}
 {id=WavePlayer_1, properties {soundWaveAssetPtr=&/FirePetalSnowGrenade_Throw_B_02}}
 {id=WavePlayer_2, properties {soundWaveAssetPtr=&/FirePetalSnowGrenade_Throw_B_03}}
 {id=WavePlayer_6, properties {soundWaveAssetPtr=&/FirePetalSnowGrenade_Throw_01}}
 {id=WavePlayer_7, properties {soundWaveAssetPtr=&/FirePetalSnowGrenade_Throw_02}}
 {id=WavePlayer_8, properties {soundWaveAssetPtr=&/FirePetalSnowGrenade_Throw_03}}
 {id=WavePlayer_9, properties {bLooping=true, soundWaveAssetPtr=&/FirePetalSnowGrenade_InAirLoop_01}}
]
properties {
 attenuationSettings=/Game/Sounds/Attenuation/Fort_Weapon_Projectile_InAir_Loop_Att
 bOverrideConcurrency=true
 maxDistance=13000
 soundClassObject=&Fort_WeaponRanged_Projectile_Effects
 volumeMultiplier=1
 concurrencyOverrides {maxCount=3}
}
