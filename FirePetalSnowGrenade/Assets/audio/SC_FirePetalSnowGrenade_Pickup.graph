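# Sound cue for picking up the FirePetalSnowGrenade, played in the loot collect sound class with weapon pickup attenuation.
# A mixer layers one fixed equip sound with three random picks: two from equip variant sets and one from a pickup set, each layer at its own volume.
# A root modulator randomizes pitch and volume slightly, and a concurrency override prevents new instances once the limit is reached.
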SoundCue /FirePetalSnowGrenade/Assets/audio/SC_FirePetalSnowGrenade_Pickup
root=Modulator_0
nodes [
 {id=Mixer_0, children [WavePlayer_4,Random_1,Random_2,Random_3], properties {inputVolume [0.8,0.5,1,1.2]}}
 {id=Modulator_0, children [Mixer_0], properties {pitchMin=1, volumeMax=1.1, volumeMin=0.9}}
 {id=Random_1, children [WavePlayer_5,WavePlayer_6], properties {weights [1,1]}}
 {id=Random_2, children [WavePlayer_8, WavePlayer_9, WavePlayer_10], properties {weights [1,1,1]}}
 {id=Random_3, children [WavePlayer_11, WavePlayer_12], properties {weights [1,1]}}
 {id=WavePlayer_10, properties {soundWaveAssetPtr=&/FirePetalSnowGrenade_Equip_C_01}}
 {id=WavePlayer_11, properties {soundWaveAssetPtr=&/FirePetalSnowGrenade_Pickup_B_01}}
 {id=WavePlayer_12, properties {soundWaveAssetPtr=&/FirePetalSnowGrenade_Pickup_B_02}}
 {id=WavePlayer_4, properties {soundWaveAssetPtr=&/FirePetalSnowGrenade_Equip_05}}
 {id=WavePlayer_5, properties {soundWaveAssetPtr=&/FirePetalSnowGrenade_Equip_B_01}}
 {id=WavePlayer_6, properties {soundWaveAssetPtr=&/FirePetalSnowGrenade_Equip_B_02}}
 {id=WavePlayer_8, properties {soundWaveAssetPtr=&/FirePetalSnowGrenade_Equip_C_03}}
 {id=WavePlayer_9, properties {soundWaveAssetPtr=&/FirePetalSnowGrenade_Equip_C_02}}
]
properties {
 attenuationSettings=&Fort_Weapon_Pickup_Att
 bOverrideConcurrency=true
 maxDistance=3500
 soundClassObject=&Fort_Loot_Collect
 volumeMultiplier=0.6
 concurrencyOverrides {resolutionRule=PreventNew}
}
