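# Sound cue that plays the throw sound for the Fire Petal Snow Grenade as player foley.
# A mixer layers a quieter fixed foley wave under one of three equally weighted, randomly chosen throw waves.
# The mix passes through a root Modulator at default settings for variation.
# It uses basic foley attenuation and overrides concurrency to prevent new instances once the limit is reached.
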SoundCue /FirePetalSnowGrenade/Assets/audio/SC_FirePetalSnowGrenade_Throw
root=Modulator_0
nodes [
 {id=Mixer_0, children [WavePlayer_0,Random_0], properties {inputVolume [0.25,1]}}
 {id=Modulator_0, children [Mixer_0]}
 {id=Random_0, children [WavePlayer_1,WavePlayer_2,WavePlayer_3], properties {weights [1,1,1]}}
 {id=WavePlayer_0, properties {soundWaveAssetPtr=&/FirePetalSnowGrenade_ThrowFoley_01}}
 {id=WavePlayer_1, properties {soundWaveAssetPtr=&/FirePetalSnowGrenade_Throw03}}
 {id=WavePlayer_2, properties {soundWaveAssetPtr=&/FirePetalSnowGrenade_Throw02}}
 {id=WavePlayer_3, properties {soundWaveAssetPtr=&/FirePetalSnowGrenade_Throw01}}
]
properties {
 attenuationSettings=&Foley_Attenuation_Basic_Att
 bOverrideConcurrency=true
 maxDistance=1536
 soundClassObject=&Fort_Player_Foley
 volumeMultiplier=1
 concurrencyOverrides {resolutionRule=PreventNew}
}
